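# Sound cue for a dungeon door opening, filed with the Fort dungeon trap sounds.
# A single wave player feeds the Dungeon_Door_Open wave straight to the output, with no randomization or mixing.
# It uses the shared Fort_Traps attenuation and an important in-game mission sound class, with a slightly lowered volume.
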SoundCue /Game/Sounds/Fort_Traps/Dungeon_Traps/Dungeon_Door_Open_Cue
root=WavePlayer_0
nodes [{id=WavePlayer_0, properties {soundWaveAssetPtr=&/Dungeon_Door_Open}}]
properties {
 attenuationSettings=&Fort_Traps_Att
 maxDistance=6000
 soundClassObject=&Fort_Mission_Ingame_Important_PSM
 volumeMultiplier=0.7
}
